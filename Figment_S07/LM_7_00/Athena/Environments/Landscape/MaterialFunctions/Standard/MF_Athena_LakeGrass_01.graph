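# Landscape material function that builds lake-area grass and recolors grass and water around a world-space center for Athena.
# Distance-blended grass is desaturated and tinted inside a sphere mask, with alpha driving specular and opacity, then a world-Z gradient multiplies in a bottom tint.
# A Grass_Color output applies the same masks to an input color.
# WaterOut darkens the input water material's base color within that sphere.
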MaterialFunction /Figment_S07/LM_7_00/Athena/Environments/Landscape/MaterialFunctions/Standard/MF_Athena_LakeGrass_01
graph {nodes [
  n1 = MaterialFunctionCall!(UseVertexInterp) {materialFunction=/Figment_S07/LM_7_00/Athena/Environments/Landscape/MaterialFunctions/MF_Landscape_GrassDistanceBlend}
  n2 = Desaturation(None=@n1.Grass RGB, Fraction=ScalarParameter() {defaultValue=0.3476, parameterName=GrassDesat})
  n3 = VectorParameter() {parameterName=Grass_Tint, defaultValue=<rgba 1.5 1.0893 0.2796 0>}
  n4 = SphereMask(
   A=WorldPosition().XYZ
   B=VectorParameter() {parameterName=Center, defaultValue=<rgba 66048 -43008 0 0>}.RGB
   Radius=ScalarParameter() {defaultValue=7534.4883, parameterName=Grass_ColorRadius}
   Hardness=ScalarParameter() {defaultValue=0.1775, parameterName=Grass_ColorHardness}
  ) {hardnessPercent=50}
  n5 = ScalarParameter() {defaultValue=2, group=Grass, parameterName=GrassSpecPower}
  n6 = SetMaterialAttributes!(
   MaterialAttributes
   BaseColor=LinearInterpolate(A=@n1.Grass RGB, B=Multiply(A=@n2, B=@n3.RGB), Alpha=@n4)
   Specular=LinearInterpolate(B=ScalarParameter() {defaultValue=0.15, group=Grass, parameterName=GrassSpecAmount}, Alpha=Power(Base=@n1.Grass A, Exp=@n5))
   Roughness=Constant() {r=0.9}
   Opacity=@n1.Grass A
  ) {attributeSetTypes [
    #1
    #2
    #3
    #4
   ]}
  n7 = VectorParameter() {parameterName=BottomGradient_Tint, defaultValue=<rgba 0.0537 0.3125 0.2845 0>}
  n8 = Subtract(
   A=ComponentMask(None=WorldPosition().XYZ) {b=true}
   B=ScalarParameter() {defaultValue=-1810.0659, group=Height_Gradient, parameterName=HeightGrad_Offset}
  )
  n9 = ScalarParameter() {defaultValue=229.9514, group=Height_Gradient, parameterName=HeightGrad_Scale}
  n10 = OneMinus(None=Clamp(None=Divide(A=@n8, B=@n9)))
  FunctionOutput(None=MaterialFunctionCall!(Material=@n6, NewBaseColor=@n7.RGB, Mask=@n10) {materialFunction=/Engine/Functions/MaterialLayerFunctions/MatLayerBlend_MultiplyBaseColor}.Blended Material)
  n11 = FunctionInput(Preview=VectorParameter() {parameterName=565, defaultValue=<rgba 0.0675 0.145 0.02 1>}.RGB) {bUsePreviewValueAsDefault=true, inputName=GrassColor}
  n12 = ScalarParameter() {defaultValue=0.3476, parameterName=GrassDesat_1}
  n13 = LinearInterpolate(
   A=@n11
   B=Multiply(A=Desaturation(None=@n11, Fraction=@n12), B=VectorParameter() {parameterName=Grass_Tint_1, defaultValue=<rgba 1.5 1.0893 0.2796 0>}.RGB)
   Alpha=@n4
  )
  FunctionOutput(None=LinearInterpolate(A=@n13, B=Multiply(A=@n13, B=@n7.RGB), Alpha=@n10)) {outputName=Grass_Color}
  n14 = GetMaterialAttributes(None=FunctionInput() {inputName=Water, inputType=FunctionInput_MaterialAttributes}) {attributeGetTypes [#1]}
  n15 = Multiply(A=@n14.BaseColor, B=VectorParameter() {parameterName=fdsd, defaultValue=<rgba 0.0021 0.0245 0.025 1>}.RGB)
  FunctionOutput(None=SetMaterialAttributes!(MaterialAttributes=@n14.MaterialAttributes, BaseColor=LinearInterpolate(A=@n14.BaseColor, B=@n15, Alpha=@n4)) {attributeSetTypes [#1]}) {outputName=WaterOut}
 ]}
